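# Translucent UI material for the ZombieElastic cosmetic item card, building an anti-aliased rounded-rectangle mask.
# Pixel edge-offset and outline parameters, normalized via MF_UI_PixelsToNormalized, set the size and thickness for MF_UI_CoreUI_RoundedMask, alongside four per-corner radii.
# Its shape and outline SDFs are anti-aliased, then combined with Max, the outline term inverted.
# Neither that result nor the DynamicColor parameter is shown feeding a material output.
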Material /FortniteCoreUI/Frontend/CosmeticItemCard/Materials/ZombieElastic/M_UI_ItemCard_ZombieElastic
graph {
 properties {
  bUsedWithStaticMesh=false
  blendMode=BLEND_Translucent
  materialDomain=MD_UI
  pixelDepthOffsetMode=PDOM_Legacy
  usedShadingModels=Default Lit
 }
 nodes [
  n1 = MaterialFunctionCall!(Value PX=Multiply(A=ScalarParameter() {group=Layout, parameterName=EdgeOffset_PX}) {constB=2}, UV Pixel Size, X or Y) {materialFunction=/FortUILibrary/MaterialFunctions/Basics/MF_UI_PixelsToNormalized}
  n2 = MaterialFunctionCall(Value PX=ScalarParameter() {defaultValue=-2, group=Circle, parameterName=Outline_PX}) {materialFunction=&MF_UI_PixelsToNormalized}
  n3 = ScalarParameter() {
   defaultValue=1
   desc="If enabled reads corner radius values in pixels.\r\nbool"
   group=Layout
   parameterName=UseValueInPixels
   sliderMax=1
   sortPriority=1
  }
  n4 = MaterialFunctionCall(
   Size=OneMinus(None=Add(A=@n1.Normalized, B=@n2.Normalized) {constB=0})
   Corner Radius=MaterialFunctionCall(
    R=ScalarParameter() {defaultValue=20, group=Layout, parameterName=TopLeftCornerRadius, sortPriority=4}
    G=ScalarParameter() {defaultValue=20, group=Layout, parameterName=BottomLeftCornerRadius, sortPriority=7}
    B=ScalarParameter() {defaultValue=20, group=Layout, parameterName=TopRightCornerRadius, sortPriority=5}
    A=ScalarParameter() {defaultValue=20, group=Layout, parameterName=BottomRightCornerRadius, sortPriority=6}
   ) {materialFunction=&AppendMany}.RGBA
   Radius in Pixels=@n3
   Outline Thickness=@n2.Normalized
  ) {materialFunction=&MF_UI_CoreUI_RoundedMask}
  Max(
   A=OneMinus(None=MaterialFunctionCall(DistanceField=@n4.Outline SDF) {materialFunction=&MF_UI_AntiAliasing}.Outer)
   B=MaterialFunctionCall(DistanceField=@n4.SDF) {materialFunction=&MF_UI_AntiAliasing}.Inner
  )
  VectorParameter() {parameterName=DynamicColor, defaultValue=<rgba 1 0 0.5663 0>}
 ]
}
